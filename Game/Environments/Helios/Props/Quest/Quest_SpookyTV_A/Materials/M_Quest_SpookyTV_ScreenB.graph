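# Screen material for the Helios SpookyTV quest prop, showing animated TV static and an optional video flipbook.
# Static multiplies two opposing panned noise samples driven by rounded, stepped time, then remaps them between low and high levels.
# A switch, off by default, enables the 4x4 video flipbook.
# Summed effect layers are added to albedo and separately scaled for emissive, alongside normal and mask textures on Default Lit.
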Material /Game/Environments/Helios/Props/Quest/Quest_SpookyTV_A/Materials/M_Quest_SpookyTV_ScreenB
graph {
 properties {bAutomaticallySetUsageInEditor=false, bHasPixelAnimation=true, bUsedWithNanite=true, usedShadingModels=Default Lit}
 nodes [
  n1 = Add(A=NamedRerouteUsage(), B=NamedRerouteUsage())
  Add(
   A=@n1
   B=TextureSampleParameter2D() {parameterName=Albedo, texture=/Game/Environments/Helios/Props/Quest/Quest_SpookyTV_A/Textures/T_Quest_SpookyTV_Screen_A_D}.RGB
  )
  Multiply(A=@n1, B=VectorParameter() {parameterName=EmissiveScale, defaultValue=<rgba 0.5 0.5 0.5 0>}.RGB)
  n2 = ScalarParameter() {defaultValue=0.03, parameterName=StaticLow}
  n3 = ScalarParameter() {defaultValue=2, parameterName=StaticHigh}
  n4 = ScalarParameter() {defaultValue=4.22, parameterName=StaticStepsA}
  n5 = Multiply(
   A=MaterialFunctionCall!(Time, Divisor, Mod) {materialFunction=/Game/Effects/Fort_Effects/Materials/_Functions/MF_ModuloTime}.Unclamped
   B=@n4
  )
  n6 = Panner(Coordinate=TextureCoordinate() {uTiling=0.3548, vTiling=0.3226}, Time=Divide(A=Round(None=@n5), B=@n4)) {speedX=0.44, speedY=0.01}
  n7 = ScalarParameter() {defaultValue=6.25, parameterName=StaticStepsB}
  n8 = Multiply(A=MaterialFunctionCall() {materialFunction=&MF_ModuloTime}.Unclamped, B=@n7)
  n9 = Panner(Coordinate=TextureCoordinate() {uTiling=0.2812, vTiling=0.2812}, Time=Divide(A=Round(None=@n8), B=@n7)) {speedX=-0.25, speedY=-0.034}
  n10 = Power(Base=Multiply(
    A=TextureSample(UVs=@n6) {texture=/Game/Athena/Items/Gameplay/Second_Chance/Textures/T_Van_TV_Static_01_D}.R
    B=TextureSample(UVs=@n9) {texture=&T_Van_TV_Static_01_D}.R
   )) {constExponent=0.8}
  NamedRerouteDeclaration(None=StaticSwitchParameter(True=LinearInterpolate(A=@n2, B=@n3, Alpha=@n10), False=Constant()) {defaultValue=true, parameterName=UseStatic}) {name=Static, nodeColor=<rgba 0.8516 0 1 1>}
  n11 = MaterialFunctionCall(Divisor=ScalarParameter() {defaultValue=1.8, parameterName=FlipbookSpeedDivisor}) {materialFunction=&MF_ModuloTime}
  n12 = Constant() {r=4}
  n13 = TextureObject() {texture=/Game/Environments/Helios/Props/Quest/Quest_SpookyTV_A/Textures/T_VideoSerie02}
  NamedRerouteDeclaration(None=StaticSwitchParameter(
    True=MaterialFunctionCall("Animation  Phase (0-1)"=@n11.Unclamped, Number of Rows=@n12, Number of Columns=@n12, Texture=@n13) {materialFunction=&FlipBook}.Result
    False=Constant()
   ) {parameterName=UseFlipbook}) {name=Flipbook, nodeColor=<rgba 0 1 0.5313 1>}
  TextureSampleParameter2D() {
   parameterName=Normals
   samplerType=SAMPLERTYPE_Normal
   texture=/Game/Environments/Helios/Props/Quest/Quest_SpookyTV_A/Textures/T_Quest_SpookyTV_Screen_A_A_N
  }
  TextureSampleParameter2D() {
   parameterName=Masks
   samplerType=SAMPLERTYPE_Masks
   texture=/Game/Environments/Helios/Props/Quest/Quest_SpookyTV_A/Textures/T_Quest_SpookyTV_Screen_A_A_S
  }
 ]
}
