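# Parent material for the spooky quest TV screen, layering animated video, noise and static as emissive over the shared GigaMix base material stack.
# Aspect-corrected UVs drive 4x4 flipbooks of two video series, while Fmod of floored time gates timed windows for scary frames and noise transitions.
# Triple-panned static adds tinted flicker, roughness feeds a tinted screen term, and an ElimationTime parameter defines a timed reactivity window.
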Material /Game/Environments/Helios/Props/Quest/Quest_SpookyTV_A/Materials/M_Quest_SpookyTV_Screen_A_PARENT
graph {
 properties {
  bAutomaticallySetUsageInEditor=false
  bTangentSpaceNormal=false
  bUseMaterialAttributes=true
  bUsedWithInstancedStaticMeshes=true
  bUsedWithNanite=true
  bUsedWithSplineMeshes=true
  ditheredLODTransition=true
  pixelDepthOffsetMode=PDOM_Legacy
  shadingModel=MSM_FromMaterialExpression
  usedShadingModels=Default Lit | Subsurface | Clear Coat | Two Sided Foliage
 }
 nodes [
  n1 = MaterialFunctionCall() {materialFunction=/Game/Global/Base4Layers_2022/MaterialFunctions/MF_GigaMix_2023}
  MaterialExpressionReroute_70 = Reroute(None=@n1.Result)
  n2 = ScalarParameter() {
   desc=Documentation link
   group="0 (Expand)  Documentation (RMB - Copy display name) and Debug view"
   parameterName=https://confluence.it.epicgames.com/display/FORT/Fortnite+Base+Material+2022
  }
  n3 = SetMaterialAttributes!(MaterialAttributes=@MaterialExpressionReroute_70.MaterialAttributes, World Position Offset=Multiply(A=@n2)) {attributeSetTypes [#1]}
  MaterialExpressionReroute_71 = Reroute(None=@n3)
  MaterialExpressionReroute_69 = Reroute(None=@MaterialExpressionReroute_71.Output)
  n4 = MaterialFunctionCall!(Material Attributes=MaterialFunctionCall!(In=MaterialFunctionCall(Material_In=@MaterialExpressionReroute_69.MaterialAttributes) {materialFunction=&MF_FN_DestructibleDamage}.Material_Out) {materialFunction=/Game/Athena/WinterArt/Materials/MF_WPO_WindNew}.Result) {materialFunction=/Game/Creative/Sets/MilitaryBase/Props/MF_FanSpin}
  n5 = MaterialFunctionCall!(In=MaterialFunctionCall!(
    In=MaterialFunctionCall!(In=@n4.Result, UseSphericalMappingOverride) {materialFunction=/Game/Packages/Fortress_SharedMaterials/Base_Material_Functions/MF_CurieFallbackEffect}.Result
    Custom Mask
    Custom Rotation Axis
   ) {materialFunction=/Game/Athena/WinterArt/Materials/MF_WindWPO_v2}.Result) {materialFunction=/Game/Global/Base4Layers_2022/MaterialFunctions/MF_SnowAndIceArchitecture}
  MaterialExpressionReroute_68 = Reroute(None=@n5.Result)
  Composite!(None=@MaterialExpressionReroute_68.Result) {desc=Transform to World in case we will use RVT after this point, subgraphName=Tangent to World}
  Composite!(MaterialAttributes=@n1.Result) {subgraphName=Documentation Link}
  Composite(MaterialAttributes=@MaterialExpressionReroute_71.Output) {desc=Game Shared WPO and Fire Effect, subgraphName=WPO And Shared Functions}
  MaterialExpressionReroute_1 = Reroute(None=@n12.RVT H)
  MaterialExpressionReroute_3 = Reroute(None=@MaterialExpressionReroute_68.Result)
  n6 = GetMaterialAttributes(None=@MaterialExpressionReroute_3.Input) {attributeGetTypes [#2]}
  n7 = Transform(None=@n6.Normal)
  n8 = LinearInterpolate(A=Multiply(A=@n7) {constB=-1}, B=@n7, Alpha=Saturate(None=TwoSidedSign())) {desc="Fixes normals\r\non backfaces"}
  n9 = SetMaterialAttributes!(
   MaterialAttributes=@n6.MaterialAttributes
   Normal=StaticSwitchParameter(True=MaterialProxyReplace(Realtime=@n8, MaterialProxy=@n6.Normal), False=@n8) {parameterName=Force TangentSpaceNormalsFor HLOD ProxyReplace}
  ) {attributeSetTypes [#2]}
  MaterialExpressionReroute_2 = Reroute(None=@n9)
  n10 = MaterialFunctionCall!(In=MaterialFunctionCall!(In=@MaterialExpressionReroute_2.Output) {materialFunction=/Game/ArtTools/Tech_Art/DirectFX/Assets/MaterialFunctions/MF_DirectFX}.Result) {materialFunction=/Game/ArtTools/Tech_Art/Veneer/MaterialFunctions/MF_Veneer}
  MaterialExpressionReroute_8 = Reroute(None=@n10.Result)
  n11 = MaterialFunctionCall!(AO, RVT_Height=@MaterialExpressionReroute_1."RVT_Height (S)", For Buildings=@MaterialExpressionReroute_8."For Moss (MA)", For Moss) {materialFunction=/Game/Environments/Asteria/Rocks/Materials/MaterialFunctions/MF_JungleBiome}
  MaterialExpressionReroute_0 = Reroute(None=@n11.Buildings Jungl)
  n12 = MaterialFunctionCall!(
   In=@MaterialExpressionReroute_0.Result
   "RVT Height Map (Roughness)"=GetMaterialAttributes(None=@n10.Result) {attributeGetTypes [#3]}.Roughness
  ) {materialFunction=/Game/Global/Base4Layers_2022/MaterialFunctions/MF_RVTBlending}
  Composite!(InputPin=@n12.Result) {subgraphName=Mobile Opt}
  Composite!("For Moss (MA)"=@n10.Result, "RVT_Height (S)"=@n12.RVT H) {subgraphName="Seasonal "}
  Constant()
  n13 = Constant() {r=12}
  n14 = Constant() {r=16}
  n15 = Frac(None=Multiply(A=Time(), B=ScalarParameter() {defaultValue=24, group=Default_Looping, parameterName=Noise_PlaySpeed}) {constB=0.2})
  n16 = Constant() {r=4}
  n17 = NamedRerouteUsage()
  n18 = TextureObjectParameter() {group=Default_Looping, parameterName=NoiseVariation, texture=/Game/Environments/Helios/Props/Quest/Quest_SpookyTV_A/Textures/T_NoiseFlip}
  n19 = MaterialFunctionCall(
   "Animation  Phase (0-1)"=LinearInterpolate(B=Divide(A=@n13, B=@n14), Alpha=@n15)
   Number of Rows=@n16
   Number of Columns=@n16
   Texture=@n18
   UVs=@n17
  ) {materialFunction=&FlipBook}
  Desaturation(None=@n19.Result, Fraction=Constant() {r=1})
  LinearInterpolate(Alpha=NamedRerouteUsage()) {constA=0.5}
  n20 = NamedRerouteUsage()
  n21 = Constant2Vector() {g=0.5, r=0.5}
  n22 = VectorParameter() {group=Default_Looping, parameterName=NoiseColor_Dark}
  n23 = VectorParameter() {group=Default_Looping, parameterName=NoiseColor_Light, defaultValue=<rgba 1 1 1 0>}
  n24 = TextureSampleParameter2D(UVs=Multiply(A=NamedRerouteUsage(), B=ScalarParameter() {defaultValue=0.2, group=Default_Looping, parameterName=Tilling_Noise})) {
   group=Default_Looping
   parameterName=NoiseIdle
   samplerSource=SSM_Wrap_WorldGroupSettings
   texture=/Game/Environments/Helios/Props/Quest/Quest_SpookyTV_A/Textures/T_SimpleNoise
  }
  n25 = Multiply(A=Time(), B=ScalarParameter() {defaultValue=0.2, group=Default_Looping, parameterName=NoiseSpeed})
  n26 = Frac(None=Add(A=@n24.R, B=Multiply(A=@n24.R, B=@n25)))
  NamedRerouteDeclaration(None=Multiply(A=SphereMask(A=@n20, B=@n21) {attenuationRadius=0.6, hardnessPercent=0.5}, B=LinearInterpolate(A=@n22.RGB, B=@n23.RGB, Alpha=@n26))) {name=NoiseIdle, nodeColor=<rgba 0.9453 0 1 1>}
  n27 = MaterialFunctionCall(Float2=NamedRerouteUsage()) {materialFunction=&BreakOutFloat2Components}
  n28 = Divide(A=@n27.R, B=Divide(A=Constant() {r=1920}, B=Constant() {r=1080}))
  n29 = Add(
   A=Multiply(
    A=MaterialFunctionCall(X=@n28, Y=@n27.G) {materialFunction=&MakeFloat2}.Result
    B=ScalarParameter() {defaultValue=1.5, group=Default_Looping, parameterName=DefaultTilling}
   )
   B=ComponentMask(None=VectorParameter() {group=Default_Looping, parameterName=DefaultOffset, defaultValue=<rgba 0 0 0 1>}.RGB) {g=true, r=true}
  )
  NamedRerouteDeclaration(None=@n29) {name=TV_UV, nodeColor=<rgba 0 0.7422 1 1>}
  n30 = VectorParameter() {
   group=Default_Looping
   parameterName=ConstantNoise_Color
   channelNames {a=Intensity, b="", g="", r=""}
   defaultValue=<rgba 0.2039 0.9271 0.6979 1>
  }
  n31 = NamedRerouteUsage()
  n32 = TextureSample(UVs=Panner(Coordinate=@n31) {speedX=5, speedY=0.1}) {
   samplerSource=SSM_Wrap_WorldGroupSettings
   samplerType=SAMPLERTYPE_Masks
   texture=/Game/Missions/Primary/RepairTheShelter/Textures/Cinematic_Screen_Static
  }
  n33 = TextureSample(UVs=Panner(Coordinate=@n31) {speedX=6, speedY=-0.05}) {
   samplerSource=SSM_Wrap_WorldGroupSettings
   samplerType=SAMPLERTYPE_Masks
   texture=&Cinematic_Screen_Static
  }
  n34 = TextureSample(UVs=Panner(Coordinate=@n31) {speedX=-3, speedY=-0.03}) {
   samplerSource=SSM_Wrap_WorldGroupSettings
   samplerType=SAMPLERTYPE_Masks
   texture=&Cinematic_Screen_Static
  }
  n35 = Add(A=Multiply(A=@n32.B) {constB=0.2}, B=Multiply(A=@n33.G, B=@n34.R))
  n36 = ScalarParameter() {defaultValue=3, group=TV Glitches, parameterName=Static Scale}
  n37 = Saturate(None=NamedRerouteDeclaration(None=Multiply(A=@n35, B=@n36)) {name=Static, nodeColor=<rgba 1 0.6562 0 1>})
  NamedRerouteDeclaration(None=Multiply(A=Multiply(A=@n30, B=@n30), B=Multiply(A=@n37) {constB=0.2, desc=constant static})) {name=StaticConstant, nodeColor=<rgba 0 0.7422 1 1>}
  n38 = Fmod(A=Floor(None=Time()), B=ScalarParameter() {defaultValue=30, group=Default_Looping, parameterName=TimeGap})
  n39 = ScalarParameter() {defaultValue=15, parameterName=SecondTriggerFrame}
  n40 = NamedRerouteDeclaration(None=@n39) {name=StartFrame01, nodeColor=<rgba 0 1 0.8828 1>}
  n41 = Constant() {r=1}
  n42 = Constant()
  n43 = NamedRerouteDeclaration(None=Add(A=@n39, B=NamedRerouteUsage())) {name=EndFrame02, nodeColor=<rgba 1 0 0.7031 1>}
  n44 = Constant()
  n45 = Constant() {r=1}
  NamedRerouteDeclaration(None=Multiply(
    A=If(A=@n38, B=@n40, "A > B"=@n41, "A == B"=@n41, "A < B"=@n42) {constB=15}
    B=If(A=@n38, B=@n43, "A > B"=@n44, "A == B"=@n45, "A < B"=@n45) {constB=15}
   )) {name=RandomScary, nodeColor=<rgba 0 0.25 1 1>}
  n46 = MaterialFunctionCall(In=NamedRerouteUsage(), Contrast=Constant() {r=0.4}) {materialFunction=&CheapContrast}
  n47 = VectorParameter() {group=Default_Looping, parameterName=RoughnessColorTint, defaultValue=<rgba 0.2105 0.2354 0.1912 1>}
  NamedRerouteDeclaration(None=Multiply(A=Multiply(A=@n46.Result, B=@n47.RGB)) {constB=0.1}) {name=RoughnessInfluence, nodeColor=<rgba 1 0 0.8906 1>}
  n48 = Divide(A=Constant() {r=14}, B=Constant() {r=16})
  n49 = Constant() {r=4}
  NamedRerouteDeclaration(None=MaterialFunctionCall("Animation  Phase (0-1)"=LinearInterpolate(B=@n48, Alpha=@n15), Number of Rows=@n49, Number of Columns=@n49, Texture=@n18, UVs=@n17) {materialFunction=&FlipBook}.Result) {name=NoiseAnimation, nodeColor=<rgba 1 0 0.0234 1>}
  n50 = TextureObjectParameter() {group=Default_Looping, parameterName=DefaultSerie01, texture=/Game/Environments/Helios/Props/Quest/Quest_SpookyTV_A/Textures/T_VideoSerie01}
  NamedRerouteDeclaration(None=@n50) {name=Serie01, nodeColor=<rgba 0.6641 1 0 1>}
  n51 = ScalarParameter() {defaultValue=0.2, group=Default_Looping, parameterName=Default_PlaySpeed}
  NamedRerouteDeclaration(None=Divide(A=Constant() {r=1}, B=@n51)) {name=OneTurnTime01, nodeColor=<rgba 0.1953 0 1 1>}
  NamedRerouteDeclaration(None=@n19.Result) {name=NoiseIdleAnimation, nodeColor=<rgba 1 0 0.1875 1>}
  NamedRerouteDeclaration(None=@n38) {name=FmodTime, nodeColor=<rgba 0 1 0.3438 1>}
  n52 = NamedRerouteUsage()
  n53 = NamedRerouteUsage()
  n54 = Constant() {r=1}
  n55 = If(A=@n52, B=@n53, "A > B"=@n54, "A == B"=@n54, "A < B"=Constant()) {constB=15}
  n56 = Constant() {r=1}
  n57 = If(A=@n52, B=Add(A=@n53), "A > B"=Constant(), "A == B"=@n56, "A < B"=@n56) {constB=15}
  n58 = NamedRerouteUsage()
  n59 = NamedRerouteUsage()
  n60 = Constant() {r=1}
  n61 = If(A=@n58, B=Subtract(A=@n59), "A > B"=@n60, "A == B"=@n60, "A < B"=Constant()) {constB=15}
  n62 = Constant() {r=1}
  n63 = If(A=@n58, B=@n59, "A > B"=Constant(), "A == B"=@n62, "A < B"=@n62) {constB=15}
  NamedRerouteDeclaration(None=Add(A=Multiply(A=@n55, B=@n57), B=Multiply(A=@n61, B=@n63))) {name=NoiseState02, nodeColor=<rgba 1 0.0234 0 1>}
  n64 = Subtract(A=Time(), B=ScalarParameter() {defaultValue=100000000, group=Reactivity, parameterName=ElimationTime})
  n65 = ScalarParameter() {defaultValue=5, group=Reactivity, parameterName=ReactivityTime}
  n66 = OneMinus(None=Saturate(None=Divide(A=@n64, B=@n65)))
  NamedRerouteDeclaration(None=Sign(None=@n66)) {name=ReactivityAlpha, nodeColor=<rgba 0.1719 1 0 1>}
  n67 = Step(X=@n66) {constY=0.1}
  n68 = Step(Y=@n66) {constX=0.9, constY=0.1}
  NamedRerouteDeclaration(None=LinearInterpolate(A=NamedRerouteUsage(), B=NamedRerouteUsage(), Alpha=Multiply(A=@n67, B=@n68))) {name=ReactivityColor, nodeColor=<rgba 1 0 0.4922 1>}
  MaterialExpressionReroute_7 = Reroute(None=@n12.Result)
  n69 = GetMaterialAttributes(None=@MaterialExpressionReroute_7.InputPin)
  n70 = Constant() {r=0.5}
  n71 = Constant() {r=0.8}
  n72 = ShadingPathSwitch(
   Default=@MaterialExpressionReroute_7.InputPin
   Mobile=QualitySwitch(
    Default=@MaterialExpressionReroute_7.InputPin
    Low=SetMaterialAttributes!(MaterialAttributes=@n69.MaterialAttributes, Specular=@n70, Roughness=@n71) {attributeSetTypes [#4, #3]}
    High=@MaterialExpressionReroute_7.InputPin
    Medium=@MaterialExpressionReroute_7.InputPin
    Epic=@MaterialExpressionReroute_7.InputPin
   )
  )
  MaterialExpressionReroute_6 = Reroute(None=@n72)
  n73 = GetMaterialAttributes(None=MaterialFunctionCall(MA In=@MaterialExpressionReroute_6.Output) {materialFunction=&MF_CreativeMoveTool}.MA Out) {attributeGetTypes [#5, #3]}
  NamedRerouteDeclaration(None=@n73.Roughness) {name=RoughnessIN, nodeColor=<rgba 0.8516 1 0 1>}
  n74 = Multiply(A=Time(), B=@n51) {constB=0.2}
  n75 = Constant() {r=4}
  NamedRerouteDeclaration(None=MaterialFunctionCall("Animation  Phase (0-1)"=@n74, Number of Rows=@n75, Number of Columns=@n75, Texture=@n50, UVs=@n29) {materialFunction=&FlipBook}.Result) {name=ScaryScreenColor01, nodeColor=<rgba 0 1 0.6953 1>}
  NamedRerouteDeclaration(None=@n74) {name=PlaySpeed_TV, nodeColor=<rgba 0 1 0.1328 1>}
  NamedRerouteDeclaration(None=@n37) {name=SubtleNoiseAlpha, nodeColor=<rgba 1 0.5859 0 1>}
  NamedRerouteDeclaration(None=LinearInterpolate(
    A=TextureCoordinate()
    B=TextureCoordinate() {coordinateIndex=1}
    Alpha=ScalarParameter() {defaultValue=1, group=Default_Looping, parameterName=UseUV1}
   )) {name=UVs, nodeColor=<rgba 1 0.4453 0 1>}
  NamedRerouteDeclaration(None=@n30) {name=NoiseTint, nodeColor=<rgba 0.8281 0 1 1>}
  NamedRerouteUsage()
  PinBase() {reroutePins [Input {expression=&self:MaterialExpressionReroute_3}]}
  PinBase!(Output=@n9) {reroutePins [Output {expression=&self:MaterialExpressionReroute_2}]}
  PinBase!() {reroutePins [
    "For Moss (MA)" {expression=&self:MaterialExpressionReroute_8}
    "RVT_Height (S)" {expression=&self:MaterialExpressionReroute_1}
   ]}
  PinBase!(Result=@n11.Buildings Jungl) {reroutePins [Result {expression=&self:MaterialExpressionReroute_0}]}
  PinBase!() {reroutePins [MaterialAttributes {expression=&self:MaterialExpressionReroute_70}]}
  PinBase!(Output=@n3) {reroutePins [Output {expression=&self:MaterialExpressionReroute_71}]}
  PinBase!() {reroutePins [MaterialAttributes {expression=&self:MaterialExpressionReroute_69}]}
  PinBase!(Result=@n5.Result) {reroutePins [Result {expression=&self:MaterialExpressionReroute_68}]}
  PinBase!() {reroutePins [InputPin {expression=&self:MaterialExpressionReroute_7}]}
  PinBase!(Output=@n72) {reroutePins [Output {expression=&self:MaterialExpressionReroute_6}]}
  n76 = Constant() {r=4}
  n77 = TextureObjectParameter() {group=Default_Looping, parameterName=ScreenVariation, texture=/Game/Environments/Helios/Props/Quest/Quest_SpookyTV_A/Textures/T_VideoSerie02}
  n78 = LinearInterpolate(
   A=NamedRerouteUsage()
   B=MaterialFunctionCall("Animation  Phase (0-1)"=NamedRerouteUsage(), Number of Rows=@n76, Number of Columns=@n76, Texture=@n77, UVs=NamedRerouteUsage()) {materialFunction=&FlipBook}.Result
   Alpha=NamedRerouteUsage()
  )
  n79 = LinearInterpolate(A=LinearInterpolate(A=@n78, B=NamedRerouteUsage(), Alpha=NamedRerouteUsage()), B=NamedRerouteUsage(), Alpha=NamedRerouteUsage())
  n80 = Add(A=Add(A=StaticSwitchParameter(True=@n78, False=@n79) {parameterName=IgnoreNoiseState2}, B=NamedRerouteUsage()), B=NamedRerouteUsage())
  n81 = VectorParameter() {parameterName=EmissiveScale, defaultValue=<rgba 1 1 1 1>}
  SetMaterialAttributes!(MaterialAttributes=@n73.MaterialAttributes, Emissive Color=Add(A=@n73.Emissive Color, B=Multiply(A=@n80, B=@n81.RGB)), Base Color) {attributeSetTypes [#5, #6]}
 ]
}
